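# Niagara system for the Winter Ghoul player body's hand smoke, drawn as short-lived ribbon trails.
# A single emitter partitions particles into ribbons, spawns them at skeletal sockets by partition, and pushes them along world Z.
# Curves shape width, color, alpha and dynamic material parameters over particle life.
# A partition test applies a system Timeline fade to alpha, and User.bisHandHidden collapses ribbon width to zero.
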NiagaraSystem /Game/Characters/Player/Male/Medium/Bodies/M_MED_Winter_Ghoul/Effects/NS_Ghoul_HandSmoke_01
dependencies {
 usedDataInterfaces [&NiagaraDataInterfaceVectorCurve, &NiagaraDataInterfaceCurve, &NiagaraDataInterfaceSocketReader]
 usedDynamicInputs []
 usedModules [
  &self:SystemSpawnScript.SystemScriptSource.SystemScriptGraph.NiagaraNodeAssignment_0.SetVariables_#1
  &SystemState
  &Timeline
 ]
}
emitters {Hand_Trails {
  emitterUpdateScript [
   &EmitterState()
   &SpawnRate(SpawnRate=60, Use Spawn Probability=false)
  ]
  particleSpawnScript [
   &PartitionParticles(Write Partition to Ribbon ID=::NewEnumerator0)
   &InitializeParticle(
    Color=<rgba 0 0 0 0.3>
    Lifetime=0.35
    Ribbon Width=25
    Ribbon Width Mode=::NewEnumerator1
    UsePositionOffset=false
    Write Material Random=true
   )
   &SocketLocation(
    SamplePosOffset=false
    Socket Index=$Particles.PartitionParticles.Partition
    Socket Reader=$System.Socket Reader
    Socket Sampling Mode=::NewEnumerator3
    VelocitySpeedLimit=false
   )
   &AddVelocity(
    Rotation Coordinate Space=::World
    Velocity=&VectorFromCurve(VectorCurve=DataInterface() {propertyValues {
       XCurve {keys [<at 0 0 RCIM_Cubic>,<at 0.6345 0 RCIM_Cubic>,<at 1.4071 0 RCIM_Cubic>,<at 2.2541 0 RCIM_Cubic>,<at 3.1517 0 RCIM_Cubic>,<at 4 0>]}
       YCurve {keys [<at 0 0 RCIM_Cubic>,<at 0.6345 0 RCIM_Cubic>,<at 1.4071 0 RCIM_Cubic>,<at 2.2541 0 RCIM_Cubic>,<at 3.1517 0 RCIM_Cubic>,<at 4 0>]}
       ZCurve {keys [
         <at 0 70 RCIM_Cubic>
         <at 0.6345 90 RCIM_Cubic>
         <at 1.4071 70 RCIM_Cubic>
         {arriveTangent=5.7321, interpMode=RCIM_Cubic, leaveTangent=5.7321, time=2.2541, value=90}
         {arriveTangent=-17.1833, interpMode=RCIM_Cubic, leaveTangent=-17.1833, time=3.1517, value=80}
         <at 4 60>
        ]}
      }})
    Velocity Speed Scale=3
   )
   &CameraOffset(Camera Offset Amount=50)
  ]
  particleUpdateScript [
   &ScaleRibbonWidth(Ribbon Width Scale=&FloatFromCurve(FloatCurve=DataInterface() {propertyValues {Curve {keys [
         <at 0 0.6 RCIM_Cubic>
         {arriveTangent=-0.2, interpMode=RCIM_Cubic, leaveTangent=-0.2, time=0.2, value=1}
         {interpMode=RCIM_Cubic, leaveTangent=-0.7818, time=1, value=0.4}
        ]}}}))
   &ParticleState()
   &ScaleColor(
    Scale Alpha=&FloatFromCurve(FloatCurve=DataInterface() {propertyValues {Curve {keys [
         <at 0 0 RCIM_Cubic>
         {arriveTangent=2.8, interpMode=RCIM_Cubic, leaveTangent=2.8, time=0.15, value=1.25}
         {arriveTangent=-1.4706, interpMode=RCIM_Cubic, leaveTangent=-1.4706, time=0.35, value=0.98}
         {interpMode=RCIM_Cubic, leaveTangent=-2.9103, time=1, value=0}
        ]}}})
    Scale RGB=&VectorFromCurve(VectorCurve=DataInterface() {propertyValues {
       XCurve {keys [<at 0.05 3.6306 RCIM_Cubic>,<at 0.75 0.0668>]}
       YCurve {keys [<at 0.05 0.6639 RCIM_Cubic>,<at 0.75 0.329>]}
       ZCurve {keys [<at 0.05 10 RCIM_Cubic>,<at 0.75 0.7>]}
      }})
    ScaleA=true
    ScaleRGB=true
   )
   &SolveForcesAndVelocity(Clamp Velocity=false, Limit Acceleration=false)
   &DynamicMaterialParameters(
    Index 0 Param 1=&FloatFromCurve(FloatCurve=DataInterface() {propertyValues {Curve {keys [
         <at 0 0.5 RCIM_Cubic>
         {arriveTangent=-0.5, interpMode=RCIM_Cubic, leaveTangent=-0.5, time=0.1, value=1}
         {arriveTangent=-1.9934, leaveTangent=-1.9934, time=1, value=0}
        ]}}})
    Index 0 Param 2=&FloatFromCurve(FloatCurve=DataInterface() {propertyValues {Curve {keys [{arriveTangent=19.7743, interpMode=RCIM_Cubic, time=0, value=0}, <at 0.1 1 RCIM_Cubic>, <at 0.3 0>]}}})
    Index 0 Param 3=-2
    Index 0 Param 4=1.2
    Module.Index0Float0Write=true
    Module.Index0Float1Write=true
    Module.Index0Float2Write=true
    Module.Index0Float3Write=true
   )
   &ScaleColor(
    Scale Alpha=&MakeCustomFloatFromBool(
     Bool=&SetBoolByIntComparison(A=$Particles.PartitionParticles.Partition, B=1, Comparison Type=::NewEnumerator4)
     True Float=$System.Timeline.InverseNormalizedTime
     False Float=1
    )
    ScaleA=true
    ScaleRGB=false
   ) {module=ScaleColor001}
   &ScaleRibbonWidth(Ribbon Width Scale=&MakeCustomFloatFromBool(Bool=$User.bisHandHidden, True Float=0, False Float=1)) {module=ScaleRibbonWidth001}
  ]
  rendererClasses [&NiagaraRibbonRendererProperties]
  renderers [{rendererClass=&NiagaraRibbonRendererProperties}]
 }}
summary {
 systemName=NS_Ghoul_HandSmoke_01
 emitters [{emitterName=Hand_Trails, rendererClasses [&NiagaraRibbonRendererProperties]}]
 userVariables [
  User.bisADS {type=NiagaraBool}
  User.bisHandHidden {type=NiagaraBool}
 ]
}
