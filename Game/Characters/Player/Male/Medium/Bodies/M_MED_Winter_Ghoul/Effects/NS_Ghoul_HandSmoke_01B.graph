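# Niagara system named for hand smoke on the Winter Ghoul character body, though as configured it draws nothing.
# The Hand_Trails ribbon emitter is disabled, leaving only an Empty emitter with no renderers active.
# It exposes two boolean user parameters for aiming-down-sights and hand-hidden state, with no visible consumer in this file.
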NiagaraSystem /Game/Characters/Player/Male/Medium/Bodies/M_MED_Winter_Ghoul/Effects/NS_Ghoul_HandSmoke_01B
dependencies=null
emitters {Empty {rendererClasses [], renderers []}}
summary {
 systemName=NS_Ghoul_HandSmoke_01B
 emitters [
  {bEnabled=false, emitterName=Hand_Trails, rendererClasses [&NiagaraRibbonRendererProperties]}
  {emitterName=Empty, rendererClasses []}
 ]
 userVariables [
  User.bisADS {type=NiagaraBool}
  User.bisHandHidden {type=NiagaraBool}
 ]
}
