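# Physics asset for the Winter Ghoul player body's rope meshes, simulating dangling belt and leg ropes with capsule bodies.
# Two three-segment belt chains branch from a shared rope driver, and two more hang from base bones on each calf.
# Chain joints have zero twist limit and swing limits that mostly tighten toward the tips.
# Larger capsules cover both calves and the left thigh, which is locked to its calf; the right thigh has no body.
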PhysicsAsset /Game/Characters/Player/Male/Medium/Bodies/M_MED_Winter_Ghoul/Meshes/M_MED_Winter_Ghoul_Ropes_New_Physics
bodies {
 base_Leg_rope_l_01 [{shapeType=Capsule, length=0.505, radius=0.505, shapeName=base_Leg_rope_l_01_capsule, rotation=<rot 0 -90 0>}]
 base_Leg_rope_l_04 [{shapeType=Capsule, length=0.505, radius=0.505, shapeName=base_Leg_rope_l_04_capsule, rotation=<rot 0 -90 0>}]
 base_Leg_rope_r_01 [{shapeType=Capsule, length=0.505, radius=0.505, shapeName=base_Leg_rope_r_01_capsule, rotation=<rot 0 -90 0>}]
 base_Leg_rope_r_04 [{shapeType=Capsule, length=0.505, radius=0.505, shapeName=base_Leg_rope_r_04_capsule, rotation=<rot 0 -90 0>}]
 base_belt_rope_01 [{shapeType=Capsule, length=0.5644, radius=0.8206, shapeName=base_belt_rope_01_capsule, rotation=<rot 74.5959 90.0001 90.0001>}]
 base_belt_rope_04 [{shapeType=Capsule, length=0.505, radius=0.505, shapeName=base_belt_rope_04_capsule, rotation=<rot 0 -90 0>}]
 base_rope_driver [{shapeType=Capsule, length=0.505, radius=0.505, shapeName=base_rope_driver_capsule, rotation=<rot 0 -90 0>}]
 calf_l [{
   shapeType=Capsule
   length=23.6611
   radius=4.5663
   shapeName=calf_l_capsule
   center=<-21.4898 -2.074 -0.9056>
   rotation=<rot -1.1581 85.1092 -92.1154>
  }]
 calf_r [{
   shapeType=Capsule
   length=23.6611
   radius=4.5663
   shapeName=calf_l_capsule
   center=<21.4898 2.074 0.9056>
   rotation=<rot -1.1581 85.1092 87.8846>
  }]
 dyn_Leg_rope_l_01 [{shapeType=Capsule, length=0.5, radius=1.5, shapeName=dyn_Leg_rope_l_01_capsule, center=<0.7536 0 0>, rotation=<rot 90 0 0>}]
 dyn_Leg_rope_l_02 [{shapeType=Capsule, length=0.5, radius=1.5, shapeName=dyn_Leg_rope_l_02_capsule, center=<0.9735 0 0>, rotation=<rot 90 0 0>}]
 dyn_Leg_rope_l_03 [{shapeType=Capsule, length=0, radius=2, shapeName=dyn_Leg_rope_l_03_capsule, center=<0.7396 0 0>, rotation=<rot 90 0 0>}]
 dyn_Leg_rope_l_04 [{shapeType=Capsule, length=3.5, radius=1.5, shapeName=dyn_Leg_rope_l_04_capsule, center=<2 0 0>, rotation=<rot 90 0 0>}]
 dyn_Leg_rope_l_05 [{shapeType=Capsule, length=3.5, radius=1.5, shapeName=dyn_Leg_rope_l_05_capsule, center=<2 0 0>, rotation=<rot 90 0 0>}]
 dyn_Leg_rope_l_06 [{shapeType=Capsule, length=0, radius=2, shapeName=dyn_Leg_rope_l_06_capsule, center=<0.6166 0 0>, rotation=<rot 90 0 0>}]
 dyn_Leg_rope_r_01 [{shapeType=Capsule, length=1, radius=1.5, shapeName=dyn_Leg_rope_r_01_capsule, center=<2 0 0>, rotation=<rot 90 0 0>}]
 dyn_Leg_rope_r_02 [{shapeType=Capsule, length=2, radius=1.5, shapeName=dyn_Leg_rope_r_02_capsule, center=<1.1103 0 0>, rotation=<rot 90 0 0>}]
 dyn_Leg_rope_r_03 [{shapeType=Capsule, length=0, radius=2, shapeName=dyn_Leg_rope_r_03_capsule, center=<0.2316 0 0>, rotation=<rot 90 0 0>}]
 dyn_Leg_rope_r_04 [{shapeType=Capsule, length=1, radius=1.5, shapeName=dyn_Leg_rope_r_04_capsule, center=<1.1395 0 0>, rotation=<rot 90 0 0>}]
 dyn_Leg_rope_r_05 [{shapeType=Capsule, length=2.5, radius=1.5, shapeName=dyn_Leg_rope_r_05_capsule, center=<2 0 0>, rotation=<rot 90 0 0>}]
 dyn_Leg_rope_r_06 [{shapeType=Capsule, length=0, radius=2, shapeName=dyn_Leg_rope_r_06_capsule, center=<0.6565 0 0>, rotation=<rot 90 0 0>}]
 dyn_belt_rope_01 [{shapeType=Capsule, length=5, radius=1.5, shapeName=dyn_belt_rope_01_capsule, center=<2.9645 0 0>, rotation=<rot 90 0 0>}]
 dyn_belt_rope_02 [{shapeType=Capsule, length=5, radius=1.5, shapeName=dyn_belt_rope_02_capsule, center=<3.1351 0 0>, rotation=<rot 90 0 0>}]
 dyn_belt_rope_03 [{shapeType=Capsule, length=1.5, radius=2, shapeName=dyn_belt_rope_03_capsule, center=<1.0556 0 0>, rotation=<rot 90 0 0>}]
 dyn_belt_rope_04 [{shapeType=Capsule, length=1, radius=1.5, shapeName=dyn_belt_rope_04_capsule, center=<1.6645 0 0>, rotation=<rot 90 0 0>}]
 dyn_belt_rope_05 [{shapeType=Capsule, length=0.5, radius=1.5, shapeName=dyn_belt_rope_05_capsule, center=<0.2164 0 0>, rotation=<rot 90 0 0>}]
 dyn_belt_rope_06 [{shapeType=Capsule, length=0.5, radius=2, shapeName=dyn_belt_rope_06_capsule, center=<0.372 0 0>, rotation=<rot 90 0 0>}]
 dyn_rope_driver [{shapeType=Capsule, length=3, radius=2, shapeName=dyn_rope_driver_capsule, center=<0.3598 2.3772 0>, rotation=<rot 90 0 0>}]
 thigh_l [{
   shapeType=Capsule
   length=18.0108
   radius=8.4193
   shapeName=thigh_l_capsule
   center=<-6.4689 1.1745 1.691>
   rotation=<rot -25.4179 89.2517 -86.3048>
  }]
}
constraints [
 {bone1Name=dyn_rope_driver, bone2Name=base_rope_driver, swing1LimitDegrees=45, swing2LimitDegrees=45, twistLimitDegrees=0}
 {bone1Name=base_belt_rope_01, bone2Name=dyn_rope_driver, swing1LimitDegrees=25, swing2LimitDegrees=25, twistLimitDegrees=0}
 {bone1Name=dyn_belt_rope_01, bone2Name=base_belt_rope_01, swing1LimitDegrees=15, swing2LimitDegrees=25, twistLimitDegrees=0}
 {bone1Name=dyn_belt_rope_02, bone2Name=dyn_belt_rope_01, swing1LimitDegrees=10, swing2LimitDegrees=30, twistLimitDegrees=0}
 {bone1Name=dyn_belt_rope_03, bone2Name=dyn_belt_rope_02, swing1LimitDegrees=5, swing2LimitDegrees=30, twistLimitDegrees=0}
 {
  bone1Name=base_belt_rope_04
  bone2Name=dyn_rope_driver
  swing1LimitDegrees=45
  swing1Motion=Free
  swing2LimitDegrees=45
  swing2Motion=Free
  twistLimitDegrees=45
  twistMotion=Free
 }
 {bone1Name=dyn_belt_rope_04, bone2Name=base_belt_rope_04, swing1LimitDegrees=15, swing2LimitDegrees=30, twistLimitDegrees=0}
 {bone1Name=dyn_belt_rope_05, bone2Name=dyn_belt_rope_04, swing1LimitDegrees=10, swing2LimitDegrees=30, twistLimitDegrees=0}
 {bone1Name=dyn_belt_rope_06, bone2Name=dyn_belt_rope_05, swing1LimitDegrees=5, swing2LimitDegrees=30, twistLimitDegrees=0}
 {bone1Name=dyn_Leg_rope_l_01, bone2Name=base_Leg_rope_l_01, swing1LimitDegrees=25, swing2LimitDegrees=24, twistLimitDegrees=0}
 {bone1Name=dyn_Leg_rope_l_02, bone2Name=dyn_Leg_rope_l_01, swing1LimitDegrees=20, swing2LimitDegrees=30, twistLimitDegrees=0}
 {bone1Name=dyn_Leg_rope_l_03, bone2Name=dyn_Leg_rope_l_02, swing1LimitDegrees=15, swing2LimitDegrees=30, twistLimitDegrees=0}
 {bone1Name=dyn_Leg_rope_l_04, bone2Name=base_Leg_rope_l_04, swing1LimitDegrees=25, swing2LimitDegrees=24, twistLimitDegrees=0}
 {bone1Name=dyn_Leg_rope_l_05, bone2Name=dyn_Leg_rope_l_04, swing1LimitDegrees=20, swing2LimitDegrees=30, twistLimitDegrees=0}
 {bone1Name=dyn_Leg_rope_l_06, bone2Name=dyn_Leg_rope_l_05, swing1LimitDegrees=15, swing2LimitDegrees=15, twistLimitDegrees=0}
 {bone1Name=dyn_Leg_rope_r_01, bone2Name=base_Leg_rope_r_01, swing1LimitDegrees=24, swing2LimitDegrees=25, twistLimitDegrees=0}
 {bone1Name=dyn_Leg_rope_r_02, bone2Name=dyn_Leg_rope_r_01, swing1LimitDegrees=20, swing2LimitDegrees=30, twistLimitDegrees=0}
 {bone1Name=dyn_Leg_rope_r_03, bone2Name=dyn_Leg_rope_r_02, swing1LimitDegrees=15, swing2LimitDegrees=30, twistLimitDegrees=0}
 {bone1Name=dyn_Leg_rope_r_04, bone2Name=base_Leg_rope_r_04, swing1LimitDegrees=24, swing2LimitDegrees=25, twistLimitDegrees=0}
 {bone1Name=dyn_Leg_rope_r_05, bone2Name=dyn_Leg_rope_r_04, swing1LimitDegrees=20, swing2LimitDegrees=30, twistLimitDegrees=0}
 {bone1Name=dyn_Leg_rope_r_06, bone2Name=dyn_Leg_rope_r_05, swing1LimitDegrees=15, swing2LimitDegrees=30, twistLimitDegrees=0}
 {
  bone1Name=base_Leg_rope_l_04
  bone2Name=calf_l
  swing1LimitDegrees=45
  swing1Motion=Free
  swing2LimitDegrees=45
  swing2Motion=Free
  twistLimitDegrees=45
  twistMotion=Free
 }
 {
  bone1Name=base_Leg_rope_l_01
  bone2Name=calf_l
  swing1LimitDegrees=45
  swing1Motion=Free
  swing2LimitDegrees=45
  swing2Motion=Free
  twistLimitDegrees=45
  twistMotion=Free
 }
 {
  bone1Name=calf_l
  bone2Name=thigh_l
  swing1LimitDegrees=45
  swing1Motion=Locked
  swing2LimitDegrees=45
  swing2Motion=Locked
  twistLimitDegrees=45
  twistMotion=Locked
 }
 {
  bone1Name=base_Leg_rope_r_01
  bone2Name=calf_r
  swing1LimitDegrees=45
  swing1Motion=Free
  swing2LimitDegrees=45
  swing2Motion=Free
  twistLimitDegrees=45
  twistMotion=Free
 }
 {
  bone1Name=base_Leg_rope_r_04
  bone2Name=calf_r
  swing1LimitDegrees=45
  swing1Motion=Free
  swing2LimitDegrees=45
  swing2Motion=Free
  twistLimitDegrees=45
  twistMotion=Free
 }
]
