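# Physics asset for the belt mesh of the male medium Winter Ghoul player body, defining physics for its dangling rope bones.
# Two chains of three box bodies (rope 01-03 and 04-06) branch from a capsule on dyn_rope_driver.
# Each link is constrained to its predecessor with moderate swing and very tight twist limits.
# An unconstrained left-thigh capsule is included, likely as a collider for the ropes.
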PhysicsAsset /Game/Characters/Player/Male/Medium/Bodies/M_MED_Winter_Ghoul/Meshes/M_MED_Winter_Ghoul_belt_Physics
bodies {
 dyn_belt_rope_01 [{shapeType=Box, extentX=2.3221, extentY=2.4935, extentZ=4.3074, center=<3.2757 0.8731 0.3018>, rotation=<rot 26.7276 101.8443 -83.2861>}]
 dyn_belt_rope_02 [{shapeType=Box, extentX=2.6901, extentY=2.2882, extentZ=7.0511, center=<1.8968 0.545 0.8179>, rotation=<rot 29.273 108.7836 -88.1981>}]
 dyn_belt_rope_03 [{shapeType=Box, extentX=4.0629, extentY=3.1536, extentZ=6.7531, center=<1.2853 -0.0508 -0.0449>, rotation=<rot -20.7467 -118.306 6.5146>}]
 dyn_belt_rope_04 [{shapeType=Box, extentX=3.1643, extentY=2.0265, extentZ=3.5006, center=<2.0058 0.2456 -0.0648>, rotation=<rot -0.284 -76.2974 -88.6602>}]
 dyn_belt_rope_05 [{shapeType=Box, extentX=3.01, extentY=4.51, extentZ=2.01, center=<0.5961 0 0>, rotation=<rot 0 -73.0093 0>}]
 dyn_belt_rope_06 [{shapeType=Box, extentX=4.0779, extentY=4.0652, extentZ=4.3022, center=<-0.0715 -0.2938 -0.6422>, rotation=<rot -13.8542 -90 1.7462>}]
 dyn_rope_driver [{shapeType=Capsule, length=4.1404, radius=2.3643, center=<1.9338 -1.1099 2.5758>, rotation=<rot 0 -87.6586 -73.8826>}]
 thigh_l [{shapeType=Capsule, length=29.4388, radius=9.2023, center=<-9.4005 0.6753 1.5273>, rotation=<rot -18.2701 86.4126 -83.7053>}]
}
constraints [
 {bone1Name=dyn_belt_rope_01, bone2Name=dyn_rope_driver, swing1LimitDegrees=45, swing2LimitDegrees=45, twistLimitDegrees=2}
 {bone1Name=dyn_belt_rope_02, bone2Name=dyn_belt_rope_01, swing1LimitDegrees=45, swing2LimitDegrees=45, twistLimitDegrees=2}
 {bone1Name=dyn_belt_rope_03, bone2Name=dyn_belt_rope_02, swing1LimitDegrees=45, swing2LimitDegrees=45, twistLimitDegrees=2}
 {bone1Name=dyn_belt_rope_04, bone2Name=dyn_rope_driver, swing1LimitDegrees=45, swing2LimitDegrees=45, twistLimitDegrees=2}
 {bone1Name=dyn_belt_rope_05, bone2Name=dyn_belt_rope_04, swing1LimitDegrees=45, swing2LimitDegrees=45, twistLimitDegrees=2}
 {bone1Name=dyn_belt_rope_06, bone2Name=dyn_belt_rope_05, swing1LimitDegrees=45, swing2LimitDegrees=45, twistLimitDegrees=2}
]
