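# Physics asset for the rope meshes on the legs of the Winter Ghoul medium male player body, covering dangling rope bones on both calves.
# Each calf twist bone has a capsule body and anchors two separate two-link rope chains (01→02 and 04→05).
# Left rope links use capsules while right ones use boxes.
# All joints share wide swing limits and very tight twist limits, so ropes bend but barely spin.
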PhysicsAsset /Game/Characters/Player/Male/Medium/Bodies/M_MED_Winter_Ghoul/Meshes/M_MED_Winter_Ghoul_ropes_Physics
bodies {
 calf_twist_02_l [{shapeType=Capsule, length=12.5257, radius=5.4278, center=<0.7639 -2.2007 -0.8973>, rotation=<rot -0.7558 -92.7527 -85.709>}]
 calf_twist_02_r [{shapeType=Capsule, length=12.5257, radius=5.4278, center=<1.1094 2.921 0.4923>, rotation=<rot -0.7558 -92.7527 -85.709>}]
 dyn_Leg_rope_l_01 [{shapeType=Capsule, length=0.8625, radius=1.3904, center=<2.1106 0.0887 0.0439>, rotation=<rot -0.3027 -90.5355 -91.5368>}]
 dyn_Leg_rope_l_02 [{shapeType=Capsule, length=1.8625, radius=1.6404, center=<3.3934 0.0765 0.0783>, rotation=<rot -0.3027 -90.5355 -91.5368>}]
 dyn_Leg_rope_l_04 [{shapeType=Capsule, length=2.1125, radius=1.6404, center=<3.7668 0.073 0.0885>, rotation=<rot -0.3027 -90.5355 -91.5368>}]
 dyn_Leg_rope_l_05 [{shapeType=Capsule, length=5.6125, radius=1.6404, center=<4.6726 0.0644 0.1127>, rotation=<rot -0.3027 -90.5355 -91.5368>}]
 dyn_Leg_rope_r_01 [{shapeType=Box, extentX=2.2808, extentY=2.2804, extentZ=5.725, center=<3.4716 -0.148 0.7877>, rotation=<rot -0.6166 -87.7208 -93.87>}]
 dyn_Leg_rope_r_02 [{shapeType=Box, extentX=2.5142, extentY=2.9476, extentZ=5.0137, center=<4.2556 0.1721 0.7473>, rotation=<rot 0.2089 91.8915 -85.7701>}]
 dyn_Leg_rope_r_04 [{shapeType=Box, extentX=2.6029, extentY=2.5779, extentZ=4.4022, center=<2.736 -0.0884 0.1811>, rotation=<rot 0.4897 -92.2543 -92.0189>}]
 dyn_Leg_rope_r_05 [{shapeType=Box, extentX=2.6861, extentY=3.0185, extentZ=5.4524, center=<5.4689 -0.1617 -0.7278>, rotation=<rot -0.3604 88.9544 -90.9948>}]
}
constraints [
 {bone1Name=dyn_Leg_rope_r_01, bone2Name=calf_twist_02_r, swing1LimitDegrees=60, swing2LimitDegrees=60, twistLimitDegrees=2}
 {bone1Name=dyn_Leg_rope_r_04, bone2Name=calf_twist_02_r, swing1LimitDegrees=60, swing2LimitDegrees=60, twistLimitDegrees=2}
 {bone1Name=dyn_Leg_rope_r_02, bone2Name=dyn_Leg_rope_r_01, swing1LimitDegrees=60, swing2LimitDegrees=60, twistLimitDegrees=2}
 {bone1Name=dyn_Leg_rope_r_05, bone2Name=dyn_Leg_rope_r_04, swing1LimitDegrees=60, swing2LimitDegrees=60, twistLimitDegrees=2}
 {bone1Name=dyn_Leg_rope_l_04, bone2Name=calf_twist_02_l, swing1LimitDegrees=60, swing2LimitDegrees=60, twistLimitDegrees=2}
 {bone1Name=dyn_Leg_rope_l_05, bone2Name=dyn_Leg_rope_l_04, swing1LimitDegrees=60, swing2LimitDegrees=60, twistLimitDegrees=2}
 {bone1Name=dyn_Leg_rope_l_01, bone2Name=calf_twist_02_l, swing1LimitDegrees=60, swing2LimitDegrees=60, twistLimitDegrees=2}
 {bone1Name=dyn_Leg_rope_l_02, bone2Name=dyn_Leg_rope_l_01, swing1LimitDegrees=60, swing2LimitDegrees=60, twistLimitDegrees=2}
]
