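# Physics asset for the left leg ropes of the Winter Ghoul medium male player body, giving the dynamic rope bones collision shapes and joint limits.
# A capsule on the calf twist bone anchors two separate two-bone chains of box bodies, rope bones 01–02 and 04–05.
# Constraints allow moderate swing at the calf, wider swing between rope segments, and almost no twist on any joint.
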PhysicsAsset /Game/Characters/Player/Male/Medium/Bodies/M_MED_Winter_Ghoul/Meshes/M_MED_Winter_Ghoul_ropes_l_Physics
bodies {
 calf_twist_02_l [{shapeType=Capsule, length=12.8389, radius=5.5908, center=<-1.5986 -2.3282 -0.087>, rotation=<rot 0 -93.6118 -87.541>}]
 dyn_Leg_rope_l_01 [{shapeType=Box, extentX=1.3665, extentY=2.0363, extentZ=3.3446, center=<2.509 0.4044 0.8309>, rotation=<rot 3.8614 -90.2102 -95.1328>}]
 dyn_Leg_rope_l_02 [{shapeType=Box, extentX=2.01, extentY=4.51, extentZ=1.51, center=<3.8626 0.2393 0.8227>, rotation=<rot 1.6175 -91.4114 -2.4897>}]
 dyn_Leg_rope_l_04 [{shapeType=Box, extentX=2.0914, extentY=2.4694, extentZ=6.9335, center=<3.1516 -0.2096 -0.2869>, rotation=<rot -1.1967 90.988 -83.541>}]
 dyn_Leg_rope_l_05 [{shapeType=Box, extentX=2.01, extentY=7.51, extentZ=1.01, center=<6.146 0 0>, rotation=<rot 0 -90 0>}]
}
constraints [
 {bone1Name=dyn_Leg_rope_l_05, bone2Name=dyn_Leg_rope_l_04, swing1LimitDegrees=60, swing2LimitDegrees=60, twistLimitDegrees=2}
 {bone1Name=dyn_Leg_rope_l_02, bone2Name=dyn_Leg_rope_l_01, swing1LimitDegrees=60, swing2LimitDegrees=60, twistLimitDegrees=2}
 {bone1Name=dyn_Leg_rope_l_04, bone2Name=calf_twist_02_l, swing1LimitDegrees=45, swing2LimitDegrees=45, twistLimitDegrees=2}
 {bone1Name=dyn_Leg_rope_l_01, bone2Name=calf_twist_02_l, swing1LimitDegrees=45, swing2LimitDegrees=45, twistLimitDegrees=2}
]
